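; Blueprint actor for crash-site ground-cover smoke, drawn by a sprite smokestack component with optional ambient audio on BeginPlay.
; The construction script creates a dynamic material instance and pushes exposed variables for size, emissive, direction, opacity, color, loop duration and TOD amount.
; It then divides both flipbook frame rates by a clamped fraction of loop duration and clamps the results, so slower-moving sprites animate more slowly.
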(event EventBeginPlay
  (if (Variables|Default|GetPlayAudio)
    (Audio|Components|Audio|Play (Variables|Default|GetAmbientAudio) 0.0)))

(event Collision|EventActorBeginOverlap
  (params (OtherActor :type Actor%20Object%20Reference)))

(event EventTick
  (params (DeltaSeconds :type Float%20%28single-precision%29)))

(fn ConstructionScript
  (bind _returnvalue (Rendering|Material|CreateDynamicMaterialInstance (Variables|Default|GetSM_Smokestack_Sprites_01) 0 (Rendering|Material|GetMaterial (Variables|Default|GetSM_Smokestack_Sprites_01) 0) "None" :!class /Script/Engine.PrimitiveComponent))
  (Utilities|FlowControl|Sequence
    (:then_0
      (Rendering|Material|SetScalarParameterValue _returnvalue "maxSize" (Variables|Default|GetmaxSize) :!class /Script/Engine.MaterialInstanceDynamic)
      (Rendering|Material|SetScalarParameterValue _returnvalue "EmissiveMult" (Variables|Default|GetEmissiveMult) :!class /Script/Engine.MaterialInstanceDynamic)
      (Rendering|Material|SetVectorParameterValue _returnvalue "OverallDirection" (Math|Conversions|ToLinearColor(Vector) (Variables|Default|GetOverallDirection)) :!class /Script/Engine.MaterialInstanceDynamic))
    (:then_1
      ;[1 1 1] Set opacity, color and TOD
      (Rendering|Material|SetScalarParameterValue _returnvalue "OpacityMult" (Variables|Default|GetOpacityMult) :!class /Script/Engine.MaterialInstanceDynamic)
      (Rendering|Material|SetScalarParameterValue _returnvalue "LoopDuration" (Variables|Default|GetLoopDuration) :!class /Script/Engine.MaterialInstanceDynamic)
      (Rendering|Material|SetVectorParameterValue _returnvalue "MainColor" (Math|Conversions|ToLinearColor(Color) (Variables|Default|GetMainColor)) :!class /Script/Engine.MaterialInstanceDynamic)
      (Rendering|Material|SetScalarParameterValue _returnvalue "TODAmount" (Variables|Default|GetTODAmount) :!class /Script/Engine.MaterialInstanceDynamic))
    (:then_2
      ;[1 1 1] Scale animation speed of the material based on how fast the sprites move (LoopSpeed)
      (bind _returnvalue_2 (Rendering|Material|GetScalarParameterValue _returnvalue "Frame Rate Max" :!class /Script/Engine.MaterialInstanceDynamic))
      (bind _returnvalue_3 (Rendering|Material|GetScalarParameterValue _returnvalue "Frame Rate" :!class /Script/Engine.MaterialInstanceDynamic))
      (bind _returnvalue_4 (Rendering|Material|GetScalarParameterValue _returnvalue "LoopDuration" :!class /Script/Engine.MaterialInstanceDynamic))
      (Rendering|Material|SetScalarParameterValue _returnvalue "Frame Rate" (Math|Float|Clamp(Float) (/ _returnvalue_3 (Math|Float|Clamp(Float) (* _returnvalue_4 0.25) 1.0 8.0)) 8.0 64.0) :!class /Script/Engine.MaterialInstanceDynamic)
      (Rendering|Material|SetScalarParameterValue _returnvalue "Frame Rate Max" (Math|Float|Clamp(Float) (/ _returnvalue_2 (Math|Float|Clamp(Float) (* _returnvalue_4 0.25) 1.0 8.0)) 12.0 128.0) :!class /Script/Engine.MaterialInstanceDynamic))))
