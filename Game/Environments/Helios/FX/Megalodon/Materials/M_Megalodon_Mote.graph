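# Additive, two-sided lit material for Megalodon mote particles on Niagara mesh particles in the Helios environment.
# A Sprite function builds world position offset that billboards the mesh, sizing it by vertex green and pixel depth.
# A sphere mask, particle alpha and near-depth fade shape each soft dot.
# Particle color is scaled by vertex green, a flattened sphere normal map supplies rounded shading, and an emissive term darkens with distance.
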Material /Game/Environments/Helios/FX/Megalodon/Materials/M_Megalodon_Mote
graph {
 properties {
  bAutomaticallySetUsageInEditor=false
  bUsedWithNiagaraMeshParticles=true
  blendMode=BLEND_Additive
  pixelDepthOffsetMode=PDOM_Legacy
  translucencyLightingMode=TLM_VolumetricPerVertexDirectional
  twoSided=true
  usedShadingModels=Default Lit
 }
 nodes [
  ComponentMask() {g=true, r=true}
  n1 = Saturate(None=Divide(A=PixelDepth().R, B=ScalarParameter() {defaultValue=3500, parameterName=darken over distance}))
  MaterialFunctionCall(Emissive=Multiply(A=VectorParameter() {parameterName=Param_2, defaultValue=<rgba 1 1 1 0>}.RGB, B=OneMinus(None=@n1))) {materialFunction=&MF_FX_EmissiveAdjustments}
  MaterialFunctionCall(
   Normal=Multiply(
    A=TextureSample() {samplerType=SAMPLERTYPE_Normal, texture=/Game/Effects/Fort_Effects/Textures/Mask/T_Sphere_N}.RGB
    B=VectorParameter() {parameterName=Param_1, defaultValue=<rgba 1 -1 1 0>}.RGB
   )
   Flatness=ScalarParameter() {parameterName=Param_6}
  ) {materialFunction=&FlattenNormal}
  Multiply()
  Multiply(B=VertexColor().G)
  n2 = TextureCoordinate()
  n3 = Constant() {r=0.5}
  n4 = ScalarParameter() {defaultValue=0.5, parameterName=Param_1}
  n5 = ScalarParameter() {defaultValue=0.5, parameterName=Param_2}
  n6 = PixelDepth()
  n7 = ScalarParameter() {defaultValue=750, parameterName=Param_3}
  Multiply(A=Multiply(A=SphereMask(A=@n2, B=@n3, Radius=@n4, Hardness=@n5), B=ParticleColor().A), B=Saturate(None=Divide(A=@n6.R, B=@n7)))
  Multiply(
   A=ParticleColor().RGB
   B=LinearInterpolate(
    A=ScalarParameter() {defaultValue=0.06, parameterName=Param_5}
    B=ScalarParameter() {defaultValue=0.15, parameterName=Param_4}
    Alpha=VertexColor().G
   )
  )
  n8 = ParticlePositionWS()
  n9 = TextureCoordinate()
  n10 = PixelDepth()
  n11 = ScalarParameter() {defaultValue=750, parameterName=Param_3}
  n12 = VectorParameter() {defaultValue=<rgba 1 1 0 0>}
  n13 = ScalarParameter() {defaultValue=5}
  n14 = Multiply(A=@n13) {constB=0.5}
  n15 = VertexColor()
  n16 = Multiply(
   A=Saturate(None=Divide(A=@n10.R, B=@n11))
   B=Multiply(A=ComponentMask(None=@n12.RGB) {g=true, r=true}, B=LinearInterpolate(A=@n14, B=@n13, Alpha=@n15.G))
  )
  n17 = WorldPosition()
  n18 = ParticlePositionWS()
  NamedRerouteDeclaration(None=Add(
    A=MaterialFunctionCall!(Center Location=@n8, 0-1 UVs=@n9, XY scale=@n16, OverridePivot, "Normals (Optional)") {materialFunction=&Sprite}.World position offset
    B=Subtract(A=@n17.XYZ, B=@n18)
   )) {name=WPO, nodeColor=<rgba 0 1 0.7188 1>}
  NamedRerouteUsage()
  Transform()
 ]
}
